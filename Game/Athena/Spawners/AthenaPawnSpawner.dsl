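; Athena spawner actor that spawns an AI target dummy when a player pawn overlaps its box on the server.
; SpawnTargetDummy blocking-loads a soft pawn class, spawns it with a behavior tree, and copies path points and loop setting onto it.
; On death the dummy is hidden and a timer calls SpawnTargetDummy again after RespawnDelay.
; That event sits behind a DoOnce with no visible reset, which blocks repeat spawns, including respawns.
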(event Custom|SpawnTargetDummy
  (Utilities|FlowControl|DoOnce false
    (:Completed
      ;[1 1 1] CUSTOM EVENT -> SpawnTargetDummy (ke * SpawnTargetDummy)
      (bind _returnvalue (Utilities|LoadClassAssetBlocking (Variables|Default|GetPawnClassSoftReference)))
      (bind _aspawn (Utilities|Casting|CastToPawnClass _returnvalue))
      (bind _returnvalue_2 (AI|SpawnAIFromClass :PawnClass _aspawn :BehaviorTree (Variables|Default|GetBT) :Location (Transformation|GetActorLocation self) :bNoCollisionFail true))
      (bind _asathena_target_dummy (Utilities|Casting|CastToAthenaTargetDummy _returnvalue_2))
      (Variables|/Game/Athena/Spawners/AthenaTargetDummy.AthenaTargetDummy_C|SetPathPoints (Variables|Default|GetPathPoints) _asathena_target_dummy)
      (Variables|/Game/Athena/Spawners/AthenaTargetDummy.AthenaTargetDummy_C|SetPathLoops (Variables|Default|GetPathLoops) _asathena_target_dummy)
      (EventDispatchers|BindEventtoOnDied _asathena_target_dummy (EventDispatchers|CreateEvent :!event Custom|OnDied_Bind)))))

(event Custom|OnDied_Bind
  (params (DamagedActor :type Actor%20Object%20Reference) (Damage :type Float%20%28single-precision%29) (InstigatedBy :type Controller%20Object%20Reference) (DamageCauser :type Actor%20Object%20Reference) (HitLocation :type Vector) (FHitComponent :type Primitive%20Component%20Object%20Reference) (BoneName :type Name) (Momentum :type Vector))
  ;[1 1 1] CUSTOM EVENT -> OnDied_Bind
  (Rendering|SetActorHiddenInGame DamagedActor true)
  (Utilities|Time|SetTimerbyEvent (EventDispatchers|CreateEvent :!event Custom|RespawnAfterDeath) (Variables|Default|GetRespawnDelay)))

(event Custom|RespawnAfterDeath
  ;[1 1 1] CUSTOM EVENT -> RespawnAfterDeath
  (CallFunction|SpawnTargetDummy))

(event OnComponentBeginOverlap(OverlapBox)
  (params (OverlappedComponent :type Primitive%20Component%20Object%20Reference) (OtherActor :type Actor%20Object%20Reference) (OtherComp :type Primitive%20Component%20Object%20Reference) (OtherBodyIndex :type Integer) (bFromSweep :type Boolean) (SweepResult :type Hit%20Result%20Structure%20%28by%20ref%29))
  (Networking|SwitchHasAuthority
    (:Authority
      (Utilities|Casting|CastToFortPlayerPawnAthena OtherActor)
      (CallFunction|SpawnTargetDummy))))

(fn ConstructionScript)
