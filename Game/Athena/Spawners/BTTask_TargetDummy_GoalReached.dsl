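; Behavior tree task that advances an Athena target dummy along its path points when a goal is reached.
; It gets the dummy from the AI controller's pawn and increments its current path point index.
; Past the last point, a looping path resets the index to zero, while a non-looping path force-kills the dummy without DBNO, then the task finishes successfully.
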(event EventReceiveExecute
  (params (OwnerActor :type Actor%20Object%20Reference))
  (bind _asaicontroller (Utilities|Casting|CastToAIController OwnerActor))
  (bind _asathena_target_dummy (Utilities|Casting|CastToAthenaTargetDummy (Pawn|GetControlledPawn _asaicontroller)))
  (bind _result (Math|Integer|IncrementInt (Variables|/Game/Athena/Spawners/AthenaTargetDummy.AthenaTargetDummy_C|GetCurrentPathPointIndex _asathena_target_dummy))
    (:%20%20
      (if (>= _result (Utilities|Array|Length (Variables|/Game/Athena/Spawners/AthenaTargetDummy.AthenaTargetDummy_C|GetPathPoints _asathena_target_dummy)))
        (if (Variables|/Game/Athena/Spawners/AthenaTargetDummy.AthenaTargetDummy_C|GetPathLoops _asathena_target_dummy)
          (Variables|/Game/Athena/Spawners/AthenaTargetDummy.AthenaTargetDummy_C|SetCurrentPathPointIndex 0 _asathena_target_dummy)
          (else
            (AI|ForceKillNoDBNO _asathena_target_dummy)))
        (AI|BehaviorTree|FinishExecute true)
        (else
          _)))))
