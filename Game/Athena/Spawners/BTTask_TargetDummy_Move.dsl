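; Behavior tree task that chooses the next move destination for an Athena target dummy AI.
; On execute, it picks a random navigable point within the goal radius around the dummy's current path point, if it has path points.
; Otherwise it picks a random navigable point within a fixed radius of the pawn's spawn location.
; The result is written to the MovementKey blackboard vector, and the task finishes successfully without moving the pawn itself.
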(event EventReceiveExecute
  (params (OwnerActor :type Actor%20Object%20Reference))
  (bind _asaicontroller (Utilities|Casting|CastToAIController OwnerActor))
  (bind _asathena_target_dummy (Utilities|Casting|CastToAthenaTargetDummy (Pawn|GetControlledPawn _asaicontroller)))
  (if (Utilities|IsArrayNonEmpty (Variables|/Game/Athena/Spawners/AthenaTargetDummy.AthenaTargetDummy_C|GetPathPoints _asathena_target_dummy))
    (bind _randomlocation (AI|Navigation|GetRandomLocationinNavigableRadius (Transformation|GetActorLocation (Utilities|Array|Get(acopy) (Variables|/Game/Athena/Spawners/AthenaTargetDummy.AthenaTargetDummy_C|GetPathPoints _asathena_target_dummy) (Variables|/Game/Athena/Spawners/AthenaTargetDummy.AthenaTargetDummy_C|GetCurrentPathPointIndex _asathena_target_dummy))) (Variables|/Game/Athena/Spawners/AthenaTargetDummy.AthenaTargetDummy_C|GetGoalRadius _asathena_target_dummy)))
    (AI|BehaviorTree|SetBlackboardValueasVector (Variables|Default|GetMovementKey) _randomlocation)
    (else
      (bind _randomlocation_2 (AI|Navigation|GetRandomLocationinNavigableRadius (Variables|/Script/FortniteGame.FortAIPawn|GetSpawnLocation _asathena_target_dummy) 1e+03))
      (AI|BehaviorTree|SetBlackboardValueasVector (Variables|Default|GetMovementKey) _randomlocation_2)))
  (AI|BehaviorTree|FinishExecute true))
