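# Behavior tree that drives movement for a target dummy AI in Athena, using the BB_TargetDummy blackboard.
# A single Sequence runs a dummy Move task bound to the MoveLocation key, then a Fort game MoveTo toward that key, then a GoalReached task.
# The MoveTo stops at the goal with a small acceptable radius; the tree has no decorators or services.
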BehaviorTree /Game/Athena/Spawners/BT_TargetDummy
blackboard=&/BB_TargetDummy
root {
 name=Sequence
 class=BTComposite_Sequence
 children [
  BTTask_TargetDummy_C_0 {
   class=BTTask_TargetDummy_Move_C
   children []
   decorators []
   properties {
    bIgnoreRestartSelf=false
    bShowPropertyDetails=true
    customDescription=""
    nodeName=""
    movementKey {bNoneIsAllowedValue=false, selectedKeyName=MoveLocation}
    tickInterval {interval=-1}
   }
   services []
  }
  FortMoveTo {
   class=FortBTTask_GameMoveTo
   properties {
    bDeimosFlavor=false
    bFinishMoveOnOverlap=true
    bStopAtGoal=true
    nodeName=FortMoveTo
    acceptableRadius {defaultValue=5}
    bReachTestIncludesAgentRadius {defaultValue=true}
    bReachTestIncludesGoalRadius {defaultValue=true}
    blackboardKey {selectedKeyName=MoveLocation}
    filterClass {defaultValue=null}
    navFilterTag {tagName=None}
   }
  }
  BTTask_TargetDummy_GoalReached_C_0 {
   class=BTTask_TargetDummy_GoalReached_C
   children []
   decorators []
   properties {bIgnoreRestartSelf=false, bShowPropertyDetails=true, customDescription="", nodeName="", tickInterval {interval=-1}}
   services []
  }
 ]
 decorators []
 properties {children [
   {childTask=&self:BTTask_TargetDummy_C_0}
   {childTask=&self:FortBTTask_GameMoveTo_0}
   {childTask=&self:BTTask_TargetDummy_GoalReached_C_0}
  ]}
}
