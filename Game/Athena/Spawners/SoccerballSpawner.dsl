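; Soccer ball Blueprint, named as a spawner, that records the last pawn to kick the ball.
; On actor overlap it casts the other actor to FortPawn, prints a green "KICK! Overlap" debug message and stores the pawn in LastToKickBall.
; BeginPlay, Tick and the construction script are empty, and the graph holds no spawning logic.
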(event EventBeginPlay)

(event Collision|EventActorBeginOverlap
  (params (OtherActor :type Actor%20Object%20Reference))
  (bind _asfort_pawn (Utilities|Casting|CastToFortPawn OtherActor))
  (Development|PrintString "KICK! Overlap" false false "(R=0.202252,G=1.000000,B=0.000000,A=1.000000)")
  (Variables|Default|SetLastToKickBall _asfort_pawn))

(event EventTick
  (params (DeltaSeconds :type Float%20%28single-precision%29)))

(fn ConstructionScript)
